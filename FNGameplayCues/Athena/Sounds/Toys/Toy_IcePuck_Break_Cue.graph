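# Sound cue for the break sound of the ice puck toy.
# A random node picks evenly between two reused ice trap slide-impact waves, and a modulator above it shifts them up in pitch within a narrow range.
# It plays at reduced volume in the gadgets sound class, using small melee weapon attenuation and a shared blaster concurrency rule.
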SoundCue /FNGameplayCues/Athena/Sounds/Toys/Toy_IcePuck_Break_Cue
root=Modulator_0
nodes [
 {id=Modulator_0, children [Random_0], properties {pitchMax=1.3, pitchMin=1.25}}
 {id=Random_0, children [WavePlayer_2,WavePlayer_3], properties {weights [1,1]}}
 {id=WavePlayer_2, properties {soundWaveAssetPtr=/FNGameplayCues/Sounds/Fort_Traps/Ice_Trap/IceTrap_SlideImpact_01}}
 {id=WavePlayer_3, properties {soundWaveAssetPtr=/FNGameplayCues/Sounds/Fort_Traps/Ice_Trap/IceTrap_SlideImpact_02}}
]
properties {
 attenuationSettings=&Fort_Weapon_Melee_Sml_Att
 maxDistance=2800
 soundClassObject=&Fort_Gadgets
 volumeMultiplier=0.3
 concurrencyOverrides {resolutionRule=PreventNew}
 concurrencySet [/Game/Sounds/Concurrency_Rules/Blaster_Global_Concurrency]
}
