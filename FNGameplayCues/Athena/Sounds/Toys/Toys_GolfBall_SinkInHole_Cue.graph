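# Sound cue for the golf ball toy sinking into a hole, layering the sink sound with a delayed score pop.
# A Mixer combines one of two random sink waves, passed through a Modulator, with one of three pops reused from Score LTM audio.
# The pop layer plays at half volume after a fixed half-second delay.
# It uses ball-bounce attenuation and the emote music sound class.
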SoundCue /FNGameplayCues/Athena/Sounds/Toys/Toys_GolfBall_SinkInHole_Cue
root=Mixer_0
nodes [
 {id=Delay_0, children [Random_2], properties {delayMax=0.5, delayMin=0.5}}
 {id=Mixer_0, children [Delay_0,Modulator_0], properties {inputVolume [0.5,1]}}
 {id=Modulator_0, children [Random_0]}
 {id=Random_0, children [WavePlayer_0,WavePlayer_1], properties {weights [1,1]}}
 {id=Random_2, children [WavePlayer_5,WavePlayer_6,WavePlayer_7], properties {weights [1,1,1]}}
 {id=WavePlayer_0, properties {soundWaveAssetPtr=&/Toys_GolfBall_SinkInHole_21}}
 {id=WavePlayer_1, properties {soundWaveAssetPtr=&/Toys_GolfBall_SinkInHole_22}}
 {id=WavePlayer_5, properties {soundWaveAssetPtr=/Game/Athena/Sounds/LTMs/Score/ScoreLTM_InGamePop_Little_01}}
 {id=WavePlayer_6, properties {soundWaveAssetPtr=/Game/Athena/Sounds/LTMs/Score/ScoreLTM_InGamePop_Little_02}}
 {id=WavePlayer_7, properties {soundWaveAssetPtr=/Game/Athena/Sounds/LTMs/Score/ScoreLTM_InGamePop_Little_03}}
]
properties {
 attenuationSettings=/Game/Sounds/Attenuation/Fort_Ball_Bounce_Athena_Att
 maxDistance=7000
 soundClassObject=&Fort_Music_Emote
 volumeMultiplier=1.25
}
